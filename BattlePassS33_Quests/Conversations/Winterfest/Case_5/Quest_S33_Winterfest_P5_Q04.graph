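# Winterfest Case 5 quest 4 (Season 33) NPC conversation, where the player confronts Hope over reports of disturbing the peace and property damage, then apologizes.
# The opening "Wintervestigation" choice requires the NPC to have the IronLilac conversation tag and the quest to be active.
# A linear chain of single-choice speech nodes follows, ending in a Continue node.
# The final apology choice carries the entry tag as a choice tag.
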ConversationDatabase /BattlePassS33_Quests/Conversations/Winterfest/Case_5/Quest_S33_Winterfest_P5_Q04
entry_points [{tag=Conversation.Entry.Season33.WF.Quest.05.part.04, destinations [#1]}]
nodes [
 ConversationEntryPointNode_1 {
  id=#1
  class=ConversationEntryPointNode
  outputs [#2]
  properties {entryTag {tagName=Conversation.Entry.Season33.WF.Quest.05.part.04}}
 }
 Speech_C_1 {
  id=#2
  class=Speech_C
  outputs [#3]
  properties {softTaskIcon=&T_UI_NPC_Intel2, generalConfig {defaultMessage="Hey. Been a minute, what's up?"}}
  sub_nodes [
   ConversationChoiceNode_24 {class=ConversationChoiceNode, properties {defaultChoiceDisplayText=Wintervestigation}}
   FortConversationRequirement_ControllerRequirement_12 {
    class=FortConversationRequirement_ControllerRequirement
    properties {
     nodeName=Require: NPC Has Conversation Tag
     participantId {tagName=Conversation.Participant.NPC}
     requirement {class=FortControllerRequirement_NPCHasConversationEntryTag, properties {queryTag {tagName=Conversation.Entry.IronLilac}}}
    }
   }
   FortConversationRequirement_ControllerRequirement_13 {
    class=FortConversationRequirement_ControllerRequirement
    properties {
     nodeName=Require: Is Quest Active
     requirement {
      class=FortControllerRequirement_IsQuestActive
      properties {questDefinition=/BattlePassS33_Quests/Items/QuestItems/Winterfest/Case_5/Quest_S33_Winterfest_P5_Q04}
     }
    }
   }
  ]
 }
 Speech_C_2 {
  id=#3
  class=Speech_C
  outputs [#4]
  properties {
   softTaskIcon=&T_UI_NPC_Intel2
   generalConfig {defaultMessage=Damaging homes. You mean... the artwork... I've done... in MY OWN HOME?}
  }
  sub_nodes [ConversationChoiceNode_0 {
    class=ConversationChoiceNode
    properties {defaultChoiceDisplayText="Hope! I've got concerning reports you're disturbing the peace and damaging homes!"}
   }]
 }
 Speech_C_6 {
  id=#4
  class=Speech_C
  outputs [#5]
  properties {
   softTaskIcon=&T_UI_NPC_Intel2
   generalConfig {defaultMessage="And disturbing the peace?? You mean like, inspiring people to rise up against tyrants??"}
  }
  sub_nodes [ConversationChoiceNode_4 {class=ConversationChoiceNode, properties {defaultChoiceDisplayText="Oh. Yeah, I guess that's... not exactly 'damage' then."}}]
 }
 Speech_C_7 {
  id=#5
  class=Speech_C
  outputs [#6]
  properties {
   softTaskIcon=&T_UI_NPC_Intel2
   generalConfig {defaultMessage="Which I haven't even done since I got here, also."}
  }
  sub_nodes [ConversationChoiceNode_25 {class=ConversationChoiceNode, properties {defaultChoiceDisplayText=...}}]
 }
 Speech_C_8 {
  id=#6
  class=Speech_C
  outputs [#7]
  properties {softTaskIcon=&T_UI_NPC_Intel2, generalConfig {defaultMessage=I'm not mad. I'm just disappointed.}}
  sub_nodes [ConversationChoiceNode_1 {
    class=ConversationChoiceNode
    properties {
     defaultChoiceDisplayText="Oof. Yet another uselesss report. Sorry, Hope."
     choiceTags {gameplayTags [{tagName=Conversation.Entry.Season33.WF.Quest.05.part.04}]}
    }
   }]
 }
 Continue_C_2 {id=#7, class=Continue_C, properties {nodeName="", softTaskIcon=&T_UI_NPC_Intel2}}
]
