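# Conversation database for the Season 33 Winterfest Case 5 quest Q05, where the player reports a dud graffiti lead at Hope's house to the GumshoeDark-tagged detective NPC.
# A single entry tag opens a linear chain of three Speech nodes with player choices, ending in a Continue node.
# The "Wintervestigation" choice requires that NPC tag and the active quest.
# The last choice carries the entry's own tag as a choice tag.
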ConversationDatabase /BattlePassS33_Quests/Conversations/Winterfest/Case_5/Quest_S33_Winterfest_P5_Q05
entry_points [{tag=Conversation.Entry.Season33.WF.Quest.05.part.05, destinations [#1]}]
nodes [
 ConversationEntryPointNode_1 {
  id=#1
  class=ConversationEntryPointNode
  outputs [#2]
  properties {entryTag {tagName=Conversation.Entry.Season33.WF.Quest.05.part.05}}
 }
 Speech_C_1 {
  id=#2
  class=Speech_C
  outputs [#3]
  properties {softTaskIcon=&T_UI_NPC_Intel2, generalConfig {defaultMessage=Rookie... Just hit me with it.}}
  sub_nodes [
   ConversationChoiceNode_27 {class=ConversationChoiceNode, properties {defaultChoiceDisplayText=Wintervestigation}}
   FortConversationRequirement_ControllerRequirement_14 {
    class=FortConversationRequirement_ControllerRequirement
    properties {
     nodeName=Require: NPC Has Conversation Tag
     participantId {tagName=Conversation.Participant.NPC}
     requirement {class=FortControllerRequirement_NPCHasConversationEntryTag, properties {queryTag {tagName=Conversation.Entry.GumshoeDark}}}
    }
   }
   FortConversationRequirement_ControllerRequirement_15 {
    class=FortConversationRequirement_ControllerRequirement
    properties {
     nodeName=Require: Is Quest Active
     requirement {
      class=FortControllerRequirement_IsQuestActive
      properties {questDefinition=/BattlePassS33_Quests/Items/QuestItems/Winterfest/Case_5/Quest_S33_Winterfest_P5_Q05}
     }
    }
   }
  ]
 }
 Speech_C_2 {
  id=#3
  class=Speech_C
  outputs [#4]
  properties {
   softTaskIcon=&T_UI_NPC_Intel2
   generalConfig {defaultMessage="Rats. Something smells fishy, and it goes all the way to the top."}
  }
  sub_nodes [ConversationChoiceNode_28 {class=ConversationChoiceNode, properties {defaultChoiceDisplayText=The graffiti was IN Hope's own house. Another dud report.}}]
 }
 Speech_C_7 {
  id=#4
  class=Speech_C
  outputs [#5]
  properties {
   softTaskIcon=&T_UI_NPC_Intel2
   generalConfig {defaultMessage="I think I know what our last case is, rookie. I'll be in touch."}
  }
  sub_nodes [ConversationChoiceNode_9 {
    class=ConversationChoiceNode
    properties {defaultChoiceDisplayText=..., choiceTags {gameplayTags [{tagName=Conversation.Entry.Season33.WF.Quest.05.part.05}]}}
   }]
 }
 Continue_C_0 {id=#5, class=Continue_C, properties {nodeName="", softTaskIcon=&T_UI_NPC_Intel2}}
]
